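# Physics asset for the AnglePatch pickaxe skeletal mesh, a two-handed melee harvesting tool cosmetic.
# It holds a single capsule body on the Melee_TwoHanded_Handle bone, offset along the handle and rotated to align with it.
# It defines no constraints, so the whole weapon acts as one rigid collision shape.
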PhysicsAsset /BRCosmetics/Weapons/FORT_Melee/Pickaxe_AnglePatch/Meshes/AnglePatch_Axe_PhysicsAsset
bodies {Melee_TwoHanded_Handle [{shapeType=Capsule, length=62.8028, radius=17.4013, center=<0 -0.1375 39.9296>, rotation=<rot 0.0005 -90 0>}]}
constraints []
